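# Physics asset for the male medium Beacon body cosmetic, covering its dangling accessory bones.
# Sparse body shapes on pelvis, spine, thighs, right calf and forearms, plus small base boxes, form an anchor tree joined by free-swing constraints.
# Limited constraints hang a sphere-tipped ballast chain from spine_02, back knots from spine_05 bases, and bracer leaves from forearm twist bones.
# A right-calf boot leaf chain branches into paired, tightly limited side leaves.
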PhysicsAsset /BRCosmetics/Characters/Player/Male/Medium/Bodies/M_MED_Beacon/Meshes/M_MED_Beacon_Physics
bodies {
 calf_twist_02_r [
  {shapeType=Capsule, length=25.2351, radius=6.8418, center=<3.5256 0.0001 -0.88>, rotation=<rot 85.7556 0.0001 0.0001>}
  {shapeType=Capsule, length=7.3363, radius=7.3371, center=<-1.8786 2.491 0.3544>, rotation=<rot 90 0 0>}
 ]
 dyn_backKnot_Base_l [{shapeType=Box, extentX=1.01, extentY=1.01, extentZ=1.01}]
 dyn_backKnot_Base_r [{shapeType=Box, extentX=1.01, extentY=1.01, extentZ=1.01}]
 dyn_backKnot_l [{shapeType=Capsule, length=1.7836, radius=1.686, center=<1 0 0>, rotation=<rot 90 0 0>}]
 dyn_backKnot_r [{shapeType=Capsule, length=4.3492, radius=1.661, center=<1.935 0 0.0001>, rotation=<rot 90 0 0>}]
 dyn_ballast_01 [{shapeType=Capsule, length=8.2354, radius=1.3072, center=<4.1485 -0.0887 -0.0365>, rotation=<rot 0 91.3716 -89.3902>}]
 dyn_ballast_01_Base [{shapeType=Box, extentX=1.01, extentY=1.01, extentZ=1.01, rotation=<rot 0 -90 0>}]
 dyn_ballast_02 [{shapeType=Capsule, length=6.2848, radius=1.1904, center=<3 0.0345 0>, rotation=<rot 0 89.0723 -88.2088>}]
 dyn_ballast_03 [{shapeType=Sphere, radius=8.0277, center=<7.7736 0.0028 0.1264>}]
 dyn_bootLeafA_01 [{shapeType=Box, extentX=0.6487, extentY=3.4171, extentZ=8.7541, center=<4.2975 -0.9654 0.0871>, rotation=<rot 26.51 -84.0518 -84.7013>}]
 dyn_bootLeafB_01 [{shapeType=Box, extentX=0.7, extentY=3.4171, extentZ=7.336, center=<4.3176 -0.7302 -0.3797>, rotation=<rot 7.7591 -86.0443 -89.4063>}]
 dyn_bootLeafC_01 [{shapeType=Box, extentX=0.9453, extentY=3.4171, extentZ=8.7541, center=<4.2938 -0.2311 -0.0086>, rotation=<rot 0 -90.7364 -89.9077>}]
 dyn_bootLeafD_01 [{shapeType=Box, extentX=0.9453, extentY=3.4171, extentZ=8.7541, center=<4.2938 -0.2311 -0.0086>, rotation=<rot 0 -90.7364 -89.9077>}]
 dyn_bootLeafE_01 [{shapeType=Box, extentX=0.9453, extentY=3.4171, extentZ=8.7541, center=<4.2938 -0.2311 -0.0086>, rotation=<rot 0 -90.7364 -89.9077>}]
 dyn_bootLeafF_01 [{shapeType=Box, extentX=1.2004, extentY=2.6961, extentZ=6.9012, center=<3.4697 -0.2787 -0.0446>, rotation=<rot 0 -88.9528 -89.0994>}]
 dyn_bootLeaf_01 [{shapeType=Capsule, length=9.3225, radius=0.7725, center=<4 0 0>, rotation=<rot 90 0 0>}]
 dyn_bootLeaf_02 [{shapeType=Capsule, length=6.8352, radius=0.5, center=<3.4082 0 0>, rotation=<rot 90 0 0>}]
 dyn_bootLeaf_03 [{shapeType=Capsule, length=7.1891, radius=0.8091, center=<3.4404 0 0>, rotation=<rot 90 0 0>}]
 dyn_bootLeaf_04 [{shapeType=Capsule, length=4.4023, radius=0.636, center=<1.8352 -0.0001 0>, rotation=<rot 90 0 0>}]
 dyn_bootLeaf_05 [{shapeType=Box, extentX=2.4875, extentY=1.5977, extentZ=6.8898, center=<3.3765 0.1449 0.2755>, rotation=<rot 0 90.6898 -88.6158>}]
 dyn_bracerLeafA_01_l [{shapeType=Box, extentX=0.9453, extentY=3.4171, extentZ=7.2469, center=<3.5729 -0.2481 -0.0108>, rotation=<rot 0 -90.7364 -89.9077>}]
 dyn_bracerLeafB_01_l [{shapeType=Box, extentX=0.9453, extentY=3.4171, extentZ=6.1742, center=<2.25 -0.2311 -0.0086>, rotation=<rot 0 -90.7364 -89.9077>}]
 dyn_bracerLeafC_01_l [{shapeType=Box, extentX=0.9453, extentY=3.4171, extentZ=6.1742, center=<2.25 -0.2311 -0.0086>, rotation=<rot 0 -90.7364 -89.9077>}]
 dyn_bracerLeafD_01_l [{shapeType=Box, extentX=0.9453, extentY=3.4171, extentZ=6.1742, center=<2.25 -0.2311 -0.0086>, rotation=<rot 0 -90.7364 -89.9077>}]
 dyn_bracerLeafE_01_l [{shapeType=Box, extentX=0.9453, extentY=3.4171, extentZ=6.1742, center=<2.25 -0.2311 -0.0086>, rotation=<rot 0 -90.7364 -89.9077>}]
 dyn_bracerLeafF_01_l [{shapeType=Box, extentX=0.9453, extentY=3.4171, extentZ=6.1742, center=<2.25 -0.2311 -0.0086>, rotation=<rot 0 -90.7364 -89.9077>}]
 dyn_bracerLeaf_01_r [{shapeType=Box, extentX=0.9453, extentY=3.4171, extentZ=6.1742, center=<2.25 -0.2311 -0.0086>, rotation=<rot 0 -90.7364 -89.9077>}]
 dyn_bracerLeaf_02_r [{shapeType=Box, extentX=0.9453, extentY=3.4171, extentZ=8.6728, center=<3.7186 -0.2499 -0.011>, rotation=<rot 0 -90.7364 -89.9077>}]
 dyn_bracerLeaf_03_r [{shapeType=Box, extentX=0.9453, extentY=3.4171, extentZ=6.1742, center=<2.25 -0.2311 -0.0086>, rotation=<rot 0 -90.7364 -89.9077>}]
 dyn_bracerLeaf_04_r [{shapeType=Box, extentX=0.9453, extentY=3.4171, extentZ=6.1742, center=<2.25 -0.2311 -0.0086>, rotation=<rot 0 -90.7364 -89.9077>}]
 dyn_bracerLeaf_05_r [{shapeType=Box, extentX=0.9453, extentY=3.4171, extentZ=6.5964, center=<2.7617 -0.2376 -0.0095>, rotation=<rot 0 -90.7364 -89.9077>}]
 lowerarm_l [{shapeType=Capsule, length=13.7488, radius=4.6844, center=<15.2156 -0.0002 2.0185>, rotation=<rot 79.2777 0 0.0001>}]
 lowerarm_twist_01_l [{shapeType=Box, extentX=1.1, extentY=1.1, extentZ=1.1}]
 lowerarm_twist_01_r [{shapeType=Box, extentX=1.1, extentY=1.1, extentZ=1.1}]
 lowerarm_twist_02_l [{shapeType=Box, extentX=1.1, extentY=1.1, extentZ=1.1}]
 lowerarm_twist_02_r [{shapeType=Box, extentX=1.1, extentY=1.1, extentZ=1.1}]
 pelvis [{shapeType=Capsule, length=3.0136, radius=14.2206, center=<0 0.7138 0.0001>}]
 spine_01 [{shapeType=Capsule, length=13.4276, radius=10.5578, rotation=<rot 90 0 0>}]
 spine_02 [{shapeType=Box, extentX=1.1, extentY=1.1, extentZ=1.1, center=<3.1534 -0.7686 -0.4871>, rotation=<rot 42.4054 -90 -0.4178>}]
 spine_05 [{shapeType=Box, extentX=1.2, extentY=1.2, extentZ=1.2, rotation=<rot 0 -179.287 -89.8556>}]
 thigh_l [{shapeType=Capsule, length=28.7004, radius=7.7641, center=<-17.9189 0 0>, rotation=<rot 90 0 0>}]
 thigh_r [{shapeType=Capsule, length=29.2196, radius=7.7814, center=<17.9902 -0.0004 0>, rotation=<rot 90 0 0>}]
}
constraints [
 {bone1Name=dyn_bootLeaf_01, bone2Name=calf_twist_02_r, swing1LimitDegrees=12, swing2LimitDegrees=33, twistLimitDegrees=4}
 {bone1Name=dyn_bootLeaf_02, bone2Name=dyn_bootLeaf_01, swing1LimitDegrees=12, swing2LimitDegrees=33, twistLimitDegrees=4}
 {bone1Name=dyn_bootLeaf_03, bone2Name=dyn_bootLeaf_02, swing1LimitDegrees=12, swing2LimitDegrees=33, twistLimitDegrees=4}
 {bone1Name=dyn_bootLeaf_04, bone2Name=dyn_bootLeaf_03, swing1LimitDegrees=12, swing2LimitDegrees=33, twistLimitDegrees=4}
 {bone1Name=dyn_bootLeaf_05, bone2Name=dyn_bootLeaf_04, swing1LimitDegrees=12, swing2LimitDegrees=33, twistLimitDegrees=4}
 {bone1Name=dyn_bootLeafE_01, bone2Name=dyn_bootLeaf_03, swing1LimitDegrees=12, swing2LimitDegrees=4, twistLimitDegrees=4}
 {bone1Name=dyn_bootLeafF_01, bone2Name=dyn_bootLeaf_03, swing1LimitDegrees=12, swing2LimitDegrees=4, twistLimitDegrees=4}
 {bone1Name=dyn_bootLeafC_01, bone2Name=dyn_bootLeaf_02, swing1LimitDegrees=12, swing2LimitDegrees=4, twistLimitDegrees=4}
 {bone1Name=dyn_bootLeafD_01, bone2Name=dyn_bootLeaf_02, swing1LimitDegrees=12, swing2LimitDegrees=4, twistLimitDegrees=4}
 {bone1Name=dyn_bootLeafA_01, bone2Name=dyn_bootLeaf_01, swing1LimitDegrees=12, swing2LimitDegrees=4, twistLimitDegrees=4}
 {bone1Name=dyn_bootLeafB_01, bone2Name=dyn_bootLeaf_01, swing1LimitDegrees=12, swing2LimitDegrees=4, twistLimitDegrees=4}
 {
  bone1Name=dyn_ballast_01_Base
  bone2Name=spine_02
  swing1LimitDegrees=45
  swing1Motion=Free
  swing2LimitDegrees=45
  swing2Motion=Free
  twistLimitDegrees=45
  twistMotion=Free
 }
 {bone1Name=dyn_ballast_01, bone2Name=dyn_ballast_01_Base, swing1LimitDegrees=45, swing2LimitDegrees=45, twistLimitDegrees=30}
 {bone1Name=dyn_ballast_02, bone2Name=dyn_ballast_01, swing1LimitDegrees=40, swing2LimitDegrees=40, twistLimitDegrees=30}
 {bone1Name=dyn_ballast_03, bone2Name=dyn_ballast_02, swing1LimitDegrees=35, swing2LimitDegrees=35, twistLimitDegrees=35}
 {
  bone1Name=spine_02
  bone2Name=spine_01
  swing1LimitDegrees=45
  swing1Motion=Free
  swing2LimitDegrees=45
  swing2Motion=Free
  twistLimitDegrees=45
  twistMotion=Free
 }
 {
  bone1Name=calf_twist_02_r
  bone2Name=thigh_r
  swing1LimitDegrees=45
  swing1Motion=Free
  swing2LimitDegrees=45
  swing2Motion=Free
  twistLimitDegrees=45
  twistMotion=Free
 }
 {
  bone1Name=spine_01
  bone2Name=pelvis
  swing1LimitDegrees=45
  swing1Motion=Free
  swing2LimitDegrees=45
  swing2Motion=Free
  twistLimitDegrees=45
  twistMotion=Free
 }
 {
  bone1Name=thigh_l
  bone2Name=pelvis
  swing1LimitDegrees=45
  swing1Motion=Free
  swing2LimitDegrees=45
  swing2Motion=Free
  twistLimitDegrees=45
  twistMotion=Free
 }
 {
  bone1Name=thigh_r
  bone2Name=pelvis
  swing1LimitDegrees=45
  swing1Motion=Free
  swing2LimitDegrees=45
  swing2Motion=Free
  twistLimitDegrees=45
  twistMotion=Free
 }
 {
  bone1Name=lowerarm_twist_01_l
  bone2Name=lowerarm_l
  swing1LimitDegrees=45
  swing1Motion=Free
  swing2LimitDegrees=45
  swing2Motion=Free
  twistLimitDegrees=45
  twistMotion=Free
 }
 {
  bone1Name=lowerarm_twist_02_l
  bone2Name=lowerarm_l
  swing1LimitDegrees=45
  swing1Motion=Free
  swing2LimitDegrees=45
  swing2Motion=Free
  twistLimitDegrees=45
  twistMotion=Free
 }
 {bone1Name=dyn_bracerLeafD_01_l, bone2Name=lowerarm_twist_01_l, swing1LimitDegrees=12, swing2LimitDegrees=4, twistLimitDegrees=4}
 {bone1Name=dyn_bracerLeafE_01_l, bone2Name=lowerarm_twist_01_l, swing1LimitDegrees=12, swing2LimitDegrees=4, twistLimitDegrees=4}
 {bone1Name=dyn_bracerLeafF_01_l, bone2Name=lowerarm_twist_01_l, swing1LimitDegrees=12, swing2LimitDegrees=4, twistLimitDegrees=4}
 {bone1Name=dyn_bracerLeafA_01_l, bone2Name=lowerarm_twist_02_l, swing1LimitDegrees=12, swing2LimitDegrees=4, twistLimitDegrees=4}
 {bone1Name=dyn_bracerLeafB_01_l, bone2Name=lowerarm_twist_02_l, swing1LimitDegrees=12, swing2LimitDegrees=4, twistLimitDegrees=4}
 {bone1Name=dyn_bracerLeafC_01_l, bone2Name=lowerarm_twist_02_l, swing1LimitDegrees=12, swing2LimitDegrees=4, twistLimitDegrees=4}
 {
  bone1Name=lowerarm_l
  bone2Name=spine_05
  swing1LimitDegrees=45
  swing1Motion=Free
  swing2LimitDegrees=45
  swing2Motion=Free
  twistLimitDegrees=45
  twistMotion=Free
 }
 {bone1Name=dyn_backKnot_r, bone2Name=dyn_backKnot_Base_r, swing1LimitDegrees=16, swing2LimitDegrees=16, twistLimitDegrees=16}
 {bone1Name=dyn_backKnot_l, bone2Name=dyn_backKnot_Base_l, swing1LimitDegrees=16, swing2LimitDegrees=16, twistLimitDegrees=16}
 {
  bone1Name=spine_05
  bone2Name=spine_02
  swing1LimitDegrees=45
  swing1Motion=Free
  swing2LimitDegrees=45
  swing2Motion=Free
  twistLimitDegrees=45
  twistMotion=Free
 }
 {
  bone1Name=lowerarm_twist_02_r
  bone2Name=spine_05
  swing1LimitDegrees=45
  swing1Motion=Free
  swing2LimitDegrees=45
  swing2Motion=Free
  twistLimitDegrees=45
  twistMotion=Free
 }
 {
  bone1Name=lowerarm_twist_01_r
  bone2Name=spine_05
  swing1LimitDegrees=45
  swing1Motion=Free
  swing2LimitDegrees=45
  swing2Motion=Free
  twistLimitDegrees=45
  twistMotion=Free
 }
 {bone1Name=dyn_bracerLeaf_05_r, bone2Name=lowerarm_twist_01_r, swing1LimitDegrees=16, swing2LimitDegrees=6, twistLimitDegrees=4}
 {bone1Name=dyn_bracerLeaf_01_r, bone2Name=lowerarm_twist_02_r, swing1LimitDegrees=34, swing2LimitDegrees=10, twistLimitDegrees=4}
 {bone1Name=dyn_bracerLeaf_02_r, bone2Name=lowerarm_twist_02_r, swing1LimitDegrees=35, swing2LimitDegrees=6, twistLimitDegrees=4}
 {bone1Name=dyn_bracerLeaf_03_r, bone2Name=lowerarm_twist_02_r, swing1LimitDegrees=16, swing2LimitDegrees=6, twistLimitDegrees=4}
 {bone1Name=dyn_bracerLeaf_04_r, bone2Name=lowerarm_twist_02_r, swing1LimitDegrees=35, swing2LimitDegrees=6, twistLimitDegrees=4}
 {
  bone1Name=dyn_backKnot_Base_r
  bone2Name=spine_05
  swing1LimitDegrees=45
  swing1Motion=Free
  swing2LimitDegrees=45
  swing2Motion=Free
  twistLimitDegrees=45
  twistMotion=Free
 }
 {
  bone1Name=dyn_backKnot_Base_l
  bone2Name=spine_05
  swing1LimitDegrees=45
  swing1Motion=Free
  swing2LimitDegrees=45
  swing2Motion=Free
  twistLimitDegrees=45
  twistMotion=Free
 }
]
