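# Post-process style Animation Blueprint for the Beacon medium male player body, adding secondary motion to its back knot and ballast bones.
# It copies the attached parent mesh's pose, runs the Beacon Control Rig, then anchors three dyn base bones through weighted multi-bone constraints to pelvis, spine, clavicles and upper-arm twist.
# A root-space RigidBody node then simulates physics with damped component motion.
# Two other constraint nodes are unconnected, and the event graph is empty.
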AnimBlueprint /BRCosmetics/Characters/Player/Male/Medium/Bodies/M_MED_Beacon/Meshes/M_MED_Beacon_AnimBP
target_skeleton=&/M_MED_Beacon_Skeleton
graphs [
 AnimGraph {nodes [
   Constraint!(
    ComponentPose
    Alpha=1
    ConstraintWeights_0=0.2385
    ConstraintWeights_1:float=0.0573
    ConstraintWeights_2:float=0.2151
    ConstraintWeights_3:float=0.264
    ConstraintWeights_4:float=0.2252
   ) -> (Pose) {
    alphaBoolBlend {blendOption=Linear}
    constraintSetup [
     {targetBone {boneName=clavicle_r}}
     {targetBone {boneName=upperarm_twist_01_r}}
     {targetBone {boneName=spine_03}}
     {targetBone {boneName=spine_04}}
     {targetBone {boneName=spine_05}}
    ]
    constraintWeights [1,1,1,1,1]
    boneToModify=~
   }
   "Constraint - dyn_backKnot_r (TTTT)"!(
    ComponentPose
    Alpha=1
    ConstraintWeights_0=0.2804
    ConstraintWeights_1:float=0.2185
    ConstraintWeights_2:float=0.2469
    ConstraintWeights_3:float=0.2543
   ) -> (Pose) {
    alphaBoolBlend {blendOption=Linear}
    boneToModify {boneName=dyn_backKnot_r}
    constraintSetup [
     {transformType=Translation, targetBone {boneName=clavicle_r}}
     {transformType=Translation, targetBone {boneName=spine_03}}
     {transformType=Translation, targetBone {boneName=spine_04}}
     {transformType=Translation, targetBone {boneName=spine_05}}
    ]
    constraintWeights [1,1,1,1]
   }
   n1 = Control Rig(Source=Copy Pose From Mesh() {bUseAttachedParent=true}) {
    alphaBoolBlend {blendOption=Linear}
    controlRigAssetReference {blueprintRigClass=&/M_MED_Beacon_CtrlRig.M_MED_Beacon_CtrlRig_C}
   }
   n2 = "Constraint - dyn_ballast_01_Base (PPPPP)"!(
    ComponentPose=Local To Component(LocalPose=@n1)
    Alpha=1
    ConstraintWeights_0=0.2336
    ConstraintWeights_1:float=0.2749
    ConstraintWeights_2:float=0.1468
    ConstraintWeights_3:float=0.014
    ConstraintWeights_4:float=0.3307
   ) -> (Pose) {
    alphaBoolBlend {blendOption=Linear}
    boneToModify {boneName=dyn_ballast_01_Base}
    constraintSetup [
     {targetBone {boneName=pelvis}}
     {targetBone {boneName=spine_02}}
     {targetBone {boneName=spine_03}}
     {targetBone {boneName=spine_04}}
     {targetBone {boneName=spine_01}}
    ]
    constraintWeights [1,1,1,1,1]
   }
   n3 = "Constraint - dyn_backKnot_Base_l (PPPP)"!(
    ComponentPose=@n2
    Alpha=1
    ConstraintWeights_0=0.2184
    ConstraintWeights_1:float=0.2467
    ConstraintWeights_2:float=0.2541
    ConstraintWeights_3:float=0.2807
   ) -> (Pose) {
    alphaBoolBlend {blendOption=Linear}
    boneToModify {boneName=dyn_backKnot_Base_l}
    constraintSetup [{targetBone {boneName=spine_03}}, {targetBone {boneName=spine_04}}, {targetBone {boneName=spine_05}}, {targetBone {boneName=clavicle_l}}]
    constraintWeights [1,1,1,1]
   }
   n4 = "Constraint - dyn_backKnot_Base_r (PPPPP)"!(
    ComponentPose=@n3
    Alpha=1
    ConstraintWeights_0=0.2385
    ConstraintWeights_1:float=0.0573
    ConstraintWeights_2:float=0.2151
    ConstraintWeights_3:float=0.264
    ConstraintWeights_4:float=0.2252
   ) -> (Pose) {
    alphaBoolBlend {blendOption=Linear}
    boneToModify {boneName=dyn_backKnot_Base_r}
    constraintSetup [
     {targetBone {boneName=clavicle_r}}
     {targetBone {boneName=upperarm_twist_01_r}}
     {targetBone {boneName=spine_03}}
     {targetBone {boneName=spine_04}}
     {targetBone {boneName=spine_05}}
    ]
    constraintWeights [1,1,1,1,1]
   }
   n5 = RigidBody(ComponentPose=@n4) {
    bClampLinearTranslationLimitToRefPose=true
    bOverrideWorldGravity=true
    overridePhysicsAsset=&/M_MED_Beacon_Physics
    simulationSpace=BaseBoneSpace
    alphaBoolBlend {blendOption=Linear}
    baseBoneRef {boneName=root}
    componentAppliedLinearAccClamp=<5000 5000 5000>
    componentLinearAccScale=<0.5 0.5 0.5>
    componentLinearVelScale=<0.25 0.25 0.25>
    overrideWorldGravity {z=-980}
   }
   Output Pose(Result=Component To Local(ComponentPose=@n5)) {layerGroup=None}
  ]}
 EventGraph {nodes [Event BlueprintUpdateAnimation()]}
]
variables []
